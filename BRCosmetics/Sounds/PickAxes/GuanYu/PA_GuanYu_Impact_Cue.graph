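# Sound cue for the impact sound of the GuanYu cosmetic pickaxe.
# A random node picks one of four equally weighted impact waves and feeds a modulator that slightly varies pitch.
# It primes on load, plays in the impacts sound class at reduced volume, and references shared pickaxe impact attenuation, a cosmetic pickaxe property sheet and player melee weapon concurrency.
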SoundCue /BRCosmetics/Sounds/PickAxes/GuanYu/PA_GuanYu_Impact_Cue
root=Modulator_0
nodes [
 {id=Modulator_0, children [Random_2], properties {pitchMax=1.1, pitchMin=0.9}}
 {id=Random_2, children [WavePlayer_5, WavePlayer_9, WavePlayer_10, WavePlayer_11], properties {weights [1,1,1,1]}}
 {id=WavePlayer_10, properties {soundWaveAssetPtr=&/PA_Guan_Yu_Impact_03}}
 {id=WavePlayer_11, properties {soundWaveAssetPtr=&/PA_Guan_Yu_Impact_04}}
 {id=WavePlayer_5, properties {soundWaveAssetPtr=&/PA_Guan_Yu_Impact_01}}
 {id=WavePlayer_9, properties {soundWaveAssetPtr=&/PA_Guan_Yu_Impact_02}}
]
properties {
 attenuationSettings=/Game/Sounds/Attenuation/Fort_Pickaxe_Impact_Att
 audioPropertiesSheet=/BRCosmetics/Sounds/Settings/PropertySheets/Pickaxe/APS_Cosmetic_Pickaxe_Impact
 bPrimeOnLoad=true
 maxDistance=4300
 soundClassObject=&Fort_Impacts
 volumeMultiplier=0.675
 attenuationOverrides {
  bAttenuateWithLPF=true
  distanceAlgorithm=Logarithmic
  falloffDistance=800
  lPFRadiusMax=800
  lPFRadiusMin=400
  attenuationShapeExtents {x=300}
 }
 concurrencyOverrides {bLimitToOwner=true, maxCount=2, resolutionRule=StopOldest}
 concurrencySet [&Player_Melee_Weapon_Conc]
}
